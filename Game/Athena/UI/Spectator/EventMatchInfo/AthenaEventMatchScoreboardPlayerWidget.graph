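# Player row widget for the spectator event match scoreboard, showing name, health/shield bars, materials, eliminations and place.
# Player-data events from the base class update the texts and feed both bars; place 0 or 999 displays a dash.
# Death or disconnect tints the row and switches bars to ELIMINATED; only a not-dead event restores them.
# DBNO hides the shield icon, runtime PreConstruct clears numeric texts, and SetRowBordersVisibility toggles three column borders.
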WidgetBlueprint /Game/Athena/UI/Spectator/EventMatchInfo/AthenaEventMatchScoreboardPlayerWidget
graphs [
 EventGraph {nodes [
   n1 = Event OnMaterialCountChanged!() -> (OutputDelegate, then, InPlayerData:Athena Remote Player View Data Object Reference, InCount:int=0) {is=title K2Node_Event}
   > "SetText (Text)"(self=Get Materials() -> (Materials:Text Object Reference), InText="To Text (Integer)"(Value=@n1.InCount))
   n2 = Event OnPlaceChanged!() -> (OutputDelegate, then, InPlayerData:Athena Remote Player View Data Object Reference, InPlace:int=0) {is=title K2Node_Event}
   n3 = "Equal (Integer)"(A=@n2.InPlace, B=0) {is=title K2Node_CallFunction}
   n4 = "Equal (Integer)"(A=@n2.InPlace, B=999)
   > "SetText (Text)"(
    self=Get Place() -> (Place:Text Object Reference)
    InText=Select(Option 0:Text="To Text (Integer)"(Value=@n2.InPlace), Option 1:Text="INVTEXT(\"-\")", Index=OR Boolean(A=@n3, B=@n4)) -> (ReturnValue:Text)
   )
   n5 = Event OnKillScoreChanged!() -> (OutputDelegate, then, InPlayerData:Athena Remote Player View Data Object Reference, InKillScore:int=0) {is=title K2Node_Event}
   > "SetText (Text)"(self=Get Eliminations() -> (Eliminations:Text Object Reference), InText="To Text (Integer)"(Value=@n5.InKillScore))
   n6 = Event OnPlayerDataChanged!() -> (OutputDelegate, then, InPlayerData:Athena Remote Player View Data Object Reference) {is=title K2Node_Event}
   > SetPlayerData!(
    execute
    self:Athena Spectator Hit Point Bar Base Object Reference=[
     Get HealthBar() -> (HealthBar:Athena Spectator Hit Point Bar Object Reference)
     Get ShieldBar() -> (ShieldBar:Athena Spectator Hit Point Bar Object Reference)
    ]
    InPlayerData:Athena Remote Player View Data Object Reference=@n6.InPlayerData
   ) -> (then)
   K2Node_Event_2 = ClearTextWidgets() {is=title K2Node_CustomEvent}
   > "SetText (Text)"(self=[Get Materials(), Get Eliminations()], InText="INVTEXT(\"0\")")
   > "SetText (Text)"(self=Get Place(), InText="INVTEXT(\"-\")")
   n7 = Event PreConstruct()
   > n8 = Branch(Condition=@n7.IsDesignTime)
   ClearTextWidgets(execute=@n8.else) {is=title K2Node_CallFunction}
   n9 = Event OnPlayerNameChanged!() -> (OutputDelegate, then, InPlayerData:Athena Remote Player View Data Object Reference, InPlayerName:String) {is=title K2Node_Event}
   > "SetText (Text)"(self=Get Name() -> (Name:Text Object Reference), InText="To Text (String)"(InString=@n9.InPlayerName))
   n10 = Event OnDisconnectedChanged!() -> (OutputDelegate, then, InPlayerData:Athena Remote Player View Data Object Reference, bIsDisconnected:bool) {is=title K2Node_Event}
   > n11 = Branch(Condition=@n10.bIsDisconnected)
   SetBrushColor(
    execute=[@n11.then, @n13.then]
    self=Get BackgroundBorder() -> (BackgroundBorder:Common Border Object Reference)
    InBrushColor=Get EliminatedTint() -> (EliminatedTint:Linear Color Structure)
   )
   > SetActiveWidgetIndex(self=Get HealthSwitcher() -> (HealthSwitcher:Common Widget Switcher Legacy Object Reference), Index=1)
   n12 = Event OnDeadStateChanged!() -> (OutputDelegate, then, InPlayerData:Athena Remote Player View Data Object Reference, bIsDead:bool) {is=title K2Node_Event}
   > n13 = Branch(Condition=@n12.bIsDead)
   SetBrushColor!(
    execute=@n13.else
    self=Get BackgroundBorder()
    InBrushColor_R:float=0
    InBrushColor_G:float=0
    InBrushColor_B:float=0
    InBrushColor_A:float=0
   ) -> (then)
   > SetActiveWidgetIndex(self=Get HealthSwitcher(), Index=0)
   n14 = Event OnDBNOStateChanged!() -> (OutputDelegate, then, InPlayerData:Athena Remote Player View Data Object Reference, bIsDBNO:bool) {is=title K2Node_Event}
   > SetVisibility(
    self=Get ShieldImage() -> (ShieldImage:Image Object Reference)
    InVisibility=Select(Option 0:ESlateVisibility Enum=SelfHitTestInvisible, Option 1:ESlateVisibility Enum=Hidden, Index=@n14.bIsDBNO) -> (ReturnValue:ESlateVisibility Enum)
   )
  ]}
 SetRowBordersVisibility {nodes [
   n1 = SetRowBordersVisibility!() -> (then, bVisible:bool=~) {is=title K2Node_FunctionEntry}
   > SetVisibility(
    self=[
     Get MaterialsBorder() -> (MaterialsBorder:Common Border Object Reference)
     Get EliminationsBorder() -> (EliminationsBorder:Common Border Object Reference)
     Get NameBorder() -> (NameBorder:Common Border Object Reference)
    ]
    InVisibility=Select(Option 0=Hidden, Option 1=SelfHitTestInvisible, Index=@n1)
   )
  ]}
]
variables [EliminatedTint {type=LinearColor, default=<rgba 0 0 0 0.8>}]
widgets {
 info {
  inheritedWidgetCount=1
  parentClass=/Script/FortniteUI.AthenaEventMatchScoreboardPlayerWidgetBase
  rootWidgetClass=&CommonBorder
  widgetCount=29
 }
 widgets [
  BackgroundBorder {
   widgetClassPath=&CommonBorder
   bIsVariable=true
   widget=BackgroundBorder
   properties {
    background {
     bIsDynamicallyLoaded=false
     drawAs=Box
     imageType=NoImage
     mirroring=NoMirror
     resourceName=None
     resourceObject=null
     tiling=NoTile
     imageSize=<32 32>
     margin=<ltrb 0 0 0 0>
     outlineSettings {
      bUseBrushTransparency=false
      roundingType=HalfHeightRadius
      width=0
      color {colorUseRule=UseColor_Specified, specifiedColor=<rgba 0 0 0 0>}
      cornerRadii=<0 0 0 0>
     }
     tintColor {colorUseRule=UseColor_Specified, specifiedColor=<rgba 1 1 1 1>}
     uVRegion {bIsValid=false, max=<0 0>, min=<0 0>}
    }
    brushColor=<rgba 0 0 0 0>
    padding=<ltrb 0 0 0 0>
   }
  }
  SizeBox_25 {
   widgetClassPath=&SizeBox
   parent=BackgroundBorder
   widget=SizeBox_25
   properties {bOverride_HeightOverride=true, heightOverride=46}
   slot {class=BorderSlot, properties {padding=<ltrb 0 0 0 0>}}
  }
  HorizontalBox_0 {widgetClassPath=&HorizontalBox, parent=SizeBox_25, widget=HorizontalBox_0, slot {class=SizeBoxSlot}}
  NameSB {
   widgetClassPath=&SizeBox
   bIsVariable=true
   parent=HorizontalBox_0
   widget=NameSB
   properties {bOverride_WidthOverride=true, widthOverride=330}
   slot {class=HorizontalBoxSlot}
  }
  NameBorder {
   widgetClassPath=&CommonBorder
   bIsVariable=true
   parent=NameSB
   widget=NameBorder
   properties {
    horizontalAlignment=HAlign_Left
    verticalAlignment=VAlign_Center
    background {
     bIsDynamicallyLoaded=false
     drawAs=Box
     imageType=NoImage
     mirroring=NoMirror
     resourceName=None
     resourceObject=null
     tiling=NoTile
     imageSize=<32 32>
     margin=<ltrb 0 0 0 0>
     outlineSettings {
      bUseBrushTransparency=false
      roundingType=HalfHeightRadius
      width=0
      color {colorUseRule=UseColor_Specified, specifiedColor=<rgba 0 0 0 0>}
      cornerRadii=<0 0 0 0>
     }
     tintColor {colorUseRule=UseColor_Specified, specifiedColor=<rgba 1 1 1 1>}
     uVRegion {bIsValid=false, max=<0 0>, min=<0 0>}
    }
    brushColor=<rgba 0 0.0593 0.3073 0.75>
    padding=<ltrb 30 3 0 0>
   }
   slot {class=SizeBoxSlot}
  }
  Name {
   widgetClassPath=&TextBlock
   bIsVariable=true
   parent=NameBorder
   widget=Name
   properties {
    justification=Center
    text=WWWWWWWWWWWWWWW
    visibility=SelfHitTestInvisible
    font {
     bForceMonospaced=false
     bMaterialIsStencil=false
     fontMaterial=null
     fontObject=&Burbank
     letterSpacing=0
     monospacedWidth=1
     size=16
     skewAmount=0
     typefaceFontName=BurbankRegular-Black
     outlineSettings {
      bApplyOutlineToDropShadows=false
      bMiteredCorners=false
      bSeparateFillAlpha=false
      outlineMaterial=null
      outlineSize=0
      outlineColor=<rgba 0 0 0 1>
     }
    }
   }
   slot {class=BorderSlot, properties {horizontalAlignment=HAlign_Left, verticalAlignment=VAlign_Center, padding=<ltrb 30 3 0 0>}}
  }
  HealthSB {
   widgetClassPath=&SizeBox
   bIsVariable=true
   parent=HorizontalBox_0
   widget=HealthSB
   properties {bOverride_WidthOverride=true, widthOverride=260}
   slot {class=HorizontalBoxSlot}
  }
  HealthBorder {
   widgetClassPath=&CommonBorder
   parent=HealthSB
   widget=HealthBorder
   properties {
    background {
     bIsDynamicallyLoaded=false
     drawAs=Box
     imageType=NoImage
     mirroring=NoMirror
     resourceName=None
     resourceObject=null
     tiling=NoTile
     imageSize=<32 32>
     margin=<ltrb 0 0 0 0>
     outlineSettings {
      bUseBrushTransparency=false
      roundingType=HalfHeightRadius
      width=0
      color {colorUseRule=UseColor_Specified, specifiedColor=<rgba 0 0 0 0>}
      cornerRadii=<0 0 0 0>
     }
     tintColor {colorUseRule=UseColor_Specified, specifiedColor=<rgba 1 1 1 1>}
     uVRegion {bIsValid=false, max=<0 0>, min=<0 0>}
    }
    brushColor=<rgba 0 0.0593 0.3073 0.4>
    padding=<ltrb 20 5 20 5>
   }
   slot {class=SizeBoxSlot}
  }
  HealthSwitcher {
   widgetClassPath=&CommonWidgetSwitcherLegacy
   bIsVariable=true
   parent=HealthBorder
   widget=HealthSwitcher
   slot {class=BorderSlot, properties {padding=<ltrb 20 5 20 5>}}
  }
  HealthShieldBox {widgetClassPath=&VerticalBox, parent=HealthSwitcher, widget=HealthShieldBox, slot {class=WidgetSwitcherSlot}}
  HorizontalBox_2 {
   widgetClassPath=&HorizontalBox
   parent=HealthShieldBox
   widget=HorizontalBox_2
   slot {class=VerticalBoxSlot, properties {verticalAlignment=VAlign_Top, padding=<ltrb 0 0 0 3>, size {sizeRule=Fill, value=1}}}
  }
  ShieldImage {
   widgetClassPath=&Image
   bIsVariable=true
   parent=HorizontalBox_2
   widget=ShieldImage
   properties {
    visibility=SelfHitTestInvisible
    brush {
     bIsDynamicallyLoaded=false
     drawAs=Image
     imageType=NoImage
     mirroring=NoMirror
     resourceName=None
     resourceObject=/Game/Athena/HUD/Atlas/Shared/icon_shield_Sprite
     tiling=NoTile
     imageSize=<16 16>
     margin=<ltrb 0 0 0 0>
     outlineSettings {
      bUseBrushTransparency=false
      roundingType=HalfHeightRadius
      width=0
      color {colorUseRule=UseColor_Specified, specifiedColor=<rgba 0 0 0 0>}
      cornerRadii=<0 0 0 0>
     }
     tintColor {colorUseRule=UseColor_Specified, specifiedColor=<rgba 1 1 1 1>}
     uVRegion {bIsValid=false, max=<0 0>, min=<0 0>}
    }
   }
   slot {class=HorizontalBoxSlot, properties {horizontalAlignment=HAlign_Center, verticalAlignment=VAlign_Center}}
  }
  ShieldBar {
   widgetClassPath=/Game/Athena/HUD/Spectator/AthenaSpectatorHitPointBar
   bIsVariable=true
   parent=HorizontalBox_2
   widget=ShieldBar
   properties {
    barType=Shield
    shieldBackgroundColor=<rgba 0.016 0.2664 0.6445 1>
    shieldDeltaColorA=<rgba 0.0865 0.4397 0.9216 1>
    shieldDeltaColorB=<rgba 0.0865 0.4397 0.9216 1>
    shieldFillColorA=<rgba 0.0976 0.6376 1 1>
    shieldFillColorB=<rgba 0.0976 0.6376 1 1>
   }
   slot {class=HorizontalBoxSlot, properties {size {sizeRule=Fill, value=1}}}
  }
  HorizontalBox_3 {
   widgetClassPath=&HorizontalBox
   parent=HealthShieldBox
   widget=HorizontalBox_3
   slot {class=VerticalBoxSlot, properties {verticalAlignment=VAlign_Bottom, size {sizeRule=Fill, value=1}}}
  }
  HealthImage {
   widgetClassPath=&Image
   parent=HorizontalBox_3
   widget=HealthImage
   properties {
    visibility=SelfHitTestInvisible
    brush {
     bIsDynamicallyLoaded=false
     drawAs=Image
     imageType=NoImage
     mirroring=NoMirror
     resourceName=None
     resourceObject=/Game/Athena/HUD/Atlas/Shared/icon_health_Sprite
     tiling=NoTile
     imageSize=<16 16>
     margin=<ltrb 0 0 0 0>
     outlineSettings {
      bUseBrushTransparency=false
      roundingType=HalfHeightRadius
      width=0
      color {colorUseRule=UseColor_Specified, specifiedColor=<rgba 0 0 0 0>}
      cornerRadii=<0 0 0 0>
     }
     tintColor {colorUseRule=UseColor_Specified, specifiedColor=<rgba 1 1 1 1>}
     uVRegion {bIsValid=false, max=<0 0>, min=<0 0>}
    }
   }
   slot {class=HorizontalBoxSlot, properties {horizontalAlignment=HAlign_Center, verticalAlignment=VAlign_Center}}
  }
  HealthBar {
   widgetClassPath=&AthenaSpectatorHitPointBar
   bIsVariable=true
   parent=HorizontalBox_3
   widget=HealthBar
   properties {
    healthBackgroundColor=<rgba 0.0212 0.3325 0.3613 1>
    healthDeltaColorA=<rgba 1 0.1651 0.1651 1>
    healthDeltaColorB=<rgba 1 0.1651 0.1651 1>
    healthFillColorA=<rgba 0.1529 1 0.1221 1>
    healthFillColorB=<rgba 0.1529 1 0.1221 1>
   }
   slot {class=HorizontalBoxSlot, properties {size {sizeRule=Fill, value=1}}}
  }
  Eliminated {
   widgetClassPath=&TextBlock
   parent=HealthSwitcher
   widget=Eliminated
   properties {
    justification=Center
    text=ELIMINATED
    visibility=SelfHitTestInvisible
    colorAndOpacity {colorUseRule=UseColor_Specified, specifiedColor=<rgba 0 0 0 0.4>}
    font {
     bForceMonospaced=false
     bMaterialIsStencil=false
     fontMaterial=null
     fontObject=&Burbank
     letterSpacing=0
     monospacedWidth=1
     size=14
     skewAmount=0
     typefaceFontName=BurbankRegular-Black
     outlineSettings {
      bApplyOutlineToDropShadows=false
      bMiteredCorners=false
      bSeparateFillAlpha=false
      outlineMaterial=null
      outlineSize=0
      outlineColor=<rgba 0 0 0 1>
     }
    }
   }
   slot {class=WidgetSwitcherSlot, properties {horizontalAlignment=HAlign_Center, verticalAlignment=VAlign_Center, padding=<ltrb 0 4 0 0>}}
  }
  MaterialsSB {
   widgetClassPath=&SizeBox
   parent=HorizontalBox_0
   widget=MaterialsSB
   properties {bOverride_WidthOverride=true, widthOverride=150}
   slot {class=HorizontalBoxSlot}
  }
  MaterialsBorder {
   widgetClassPath=&CommonBorder
   bIsVariable=true
   parent=MaterialsSB
   widget=MaterialsBorder
   properties {
    horizontalAlignment=HAlign_Center
    verticalAlignment=VAlign_Center
    background {
     bIsDynamicallyLoaded=false
     drawAs=Box
     imageType=NoImage
     mirroring=NoMirror
     resourceName=None
     resourceObject=null
     tiling=NoTile
     imageSize=<32 32>
     margin=<ltrb 0 0 0 0>
     outlineSettings {
      bUseBrushTransparency=false
      roundingType=HalfHeightRadius
      width=0
      color {colorUseRule=UseColor_Specified, specifiedColor=<rgba 0 0 0 0>}
      cornerRadii=<0 0 0 0>
     }
     tintColor {colorUseRule=UseColor_Specified, specifiedColor=<rgba 1 1 1 1>}
     uVRegion {bIsValid=false, max=<0 0>, min=<0 0>}
    }
    brushColor=<rgba 0 0.0593 0.3073 0.6>
    padding=<ltrb 0 4 0 0>
   }
   slot {class=SizeBoxSlot}
  }
  Materials {
   widgetClassPath=&TextBlock
   bIsVariable=true
   parent=MaterialsBorder
   widget=Materials
   properties {
    justification=Center
    text=888
    visibility=SelfHitTestInvisible
    font {
     bForceMonospaced=false
     bMaterialIsStencil=false
     fontMaterial=null
     fontObject=&Burbank
     letterSpacing=0
     monospacedWidth=1
     size=22
     skewAmount=0
     typefaceFontName=BurbankRegular-Black
     outlineSettings {
      bApplyOutlineToDropShadows=false
      bMiteredCorners=false
      bSeparateFillAlpha=false
      outlineMaterial=null
      outlineSize=0
      outlineColor=<rgba 0 0 0 1>
     }
    }
   }
   slot {class=BorderSlot, properties {horizontalAlignment=HAlign_Center, verticalAlignment=VAlign_Center, padding=<ltrb 0 4 0 0>}}
  }
  EliminationsSB {
   widgetClassPath=&SizeBox
   bIsVariable=true
   parent=HorizontalBox_0
   widget=EliminationsSB
   properties {bOverride_WidthOverride=true, widthOverride=150}
   slot {class=HorizontalBoxSlot}
  }
  EliminationsBorder {
   widgetClassPath=&CommonBorder
   bIsVariable=true
   parent=EliminationsSB
   widget=EliminationsBorder
   properties {
    horizontalAlignment=HAlign_Center
    verticalAlignment=VAlign_Center
    background {
     bIsDynamicallyLoaded=false
     drawAs=Box
     imageType=NoImage
     mirroring=NoMirror
     resourceName=None
     resourceObject=null
     tiling=NoTile
     imageSize=<32 32>
     margin=<ltrb 0 0 0 0>
     outlineSettings {
      bUseBrushTransparency=false
      roundingType=HalfHeightRadius
      width=0
      color {colorUseRule=UseColor_Specified, specifiedColor=<rgba 0 0 0 0>}
      cornerRadii=<0 0 0 0>
     }
     tintColor {colorUseRule=UseColor_Specified, specifiedColor=<rgba 1 1 1 1>}
     uVRegion {bIsValid=false, max=<0 0>, min=<0 0>}
    }
    brushColor=<rgba 0 0.0593 0.3073 0.4>
    padding=<ltrb 0 0 0 0>
   }
   slot {class=SizeBoxSlot}
  }
  HorizontalBox_5 {
   widgetClassPath=&HorizontalBox
   parent=EliminationsBorder
   widget=HorizontalBox_5
   slot {class=BorderSlot, properties {horizontalAlignment=HAlign_Center, verticalAlignment=VAlign_Center, padding=<ltrb 0 0 0 0>}}
  }
  EliminationsImage {
   widgetClassPath=&Image
   parent=HorizontalBox_5
   widget=EliminationsImage
   properties {
    visibility=SelfHitTestInvisible
    brush {
     bIsDynamicallyLoaded=false
     drawAs=Image
     imageType=NoImage
     mirroring=NoMirror
     resourceName=None
     resourceObject=/Game/UI/Foundation/Textures/Icons/Replay/T-Icon-Replay-Filter-Kill-L
     tiling=NoTile
     imageSize=<22 22>
     margin=<ltrb 0 0 0 0>
     outlineSettings {
      bUseBrushTransparency=false
      roundingType=HalfHeightRadius
      width=0
      color {colorUseRule=UseColor_Specified, specifiedColor=<rgba 0 0 0 0>}
      cornerRadii=<0 0 0 0>
     }
     tintColor {colorUseRule=UseColor_Specified, specifiedColor=<rgba 1 1 1 0.9>}
     uVRegion {bIsValid=false, max=<0 0>, min=<0 0>}
    }
   }
   slot {class=HorizontalBoxSlot, properties {horizontalAlignment=HAlign_Center, verticalAlignment=VAlign_Center, padding=<ltrb 0 2 0 0>}}
  }
  Eliminations {
   widgetClassPath=&TextBlock
   bIsVariable=true
   parent=HorizontalBox_5
   widget=Eliminations
   properties {
    justification=Center
    text=888
    visibility=SelfHitTestInvisible
    font {
     bForceMonospaced=false
     bMaterialIsStencil=false
     fontMaterial=null
     fontObject=&Burbank
     letterSpacing=0
     monospacedWidth=1
     size=22
     skewAmount=0
     typefaceFontName=BurbankRegular-Black
     outlineSettings {
      bApplyOutlineToDropShadows=false
      bMiteredCorners=false
      bSeparateFillAlpha=false
      outlineMaterial=null
      outlineSize=0
      outlineColor=<rgba 0 0 0 1>
     }
    }
   }
   slot {class=HorizontalBoxSlot, properties {padding=<ltrb 0 8 0 0>}}
  }
  PlaceSB {
   widgetClassPath=&SizeBox
   bIsVariable=true
   parent=HorizontalBox_0
   widget=PlaceSB
   properties {bOverride_WidthOverride=true, widthOverride=110}
   slot {class=HorizontalBoxSlot}
  }
  PlaceBorder {
   widgetClassPath=&CommonBorder
   parent=PlaceSB
   widget=PlaceBorder
   properties {
    horizontalAlignment=HAlign_Center
    verticalAlignment=VAlign_Center
    background {
     bIsDynamicallyLoaded=false
     drawAs=Box
     imageType=NoImage
     mirroring=NoMirror
     resourceName=None
     resourceObject=null
     tiling=NoTile
     imageSize=<32 32>
     margin=<ltrb 0 0 0 0>
     outlineSettings {
      bUseBrushTransparency=false
      roundingType=HalfHeightRadius
      width=0
      color {colorUseRule=UseColor_Specified, specifiedColor=<rgba 0 0 0 0>}
      cornerRadii=<0 0 0 0>
     }
     tintColor {colorUseRule=UseColor_Specified, specifiedColor=<rgba 1 1 1 1>}
     uVRegion {bIsValid=false, max=<0 0>, min=<0 0>}
    }
    brushColor=<rgba 0 0.0593 0.3073 0.6>
    padding=<ltrb 0 4 0 0>
   }
   slot {class=SizeBoxSlot}
  }
  Place {
   widgetClassPath=&TextBlock
   bIsVariable=true
   parent=PlaceBorder
   widget=Place
   properties {
    justification=Center
    text=-
    visibility=SelfHitTestInvisible
    font {
     bForceMonospaced=false
     bMaterialIsStencil=false
     fontMaterial=null
     fontObject=&Burbank
     letterSpacing=0
     monospacedWidth=1
     size=22
     skewAmount=0
     typefaceFontName=BurbankRegular-Black
     outlineSettings {
      bApplyOutlineToDropShadows=false
      bMiteredCorners=false
      bSeparateFillAlpha=false
      outlineMaterial=null
      outlineSize=0
      outlineColor=<rgba 0 0 0 1>
     }
    }
   }
   slot {class=BorderSlot, properties {horizontalAlignment=HAlign_Center, verticalAlignment=VAlign_Center, padding=<ltrb 0 4 0 0>}}
  }
  InputActionWidget {widgetClassPath=&CommonActionWidget, bInherited=true, widget=null}
 ]
}
