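# Squad row widget for the spectator event match scoreboard, showing a squad number badge beside a vertical list of that squad's players.
# Squad ID and color events set the badge text and border color, and a visibility event collapses or shows the badge.
# Adding a player rebuilds the whole list, resetting the dynamic entry box and creating one player widget per squad data entry before invalidating layout.
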WidgetBlueprint /FortniteCoreUI/Athena/UI/Spectator/EventMatchInfo/AthenaEventMatchScoreboardSquadWidget
graphs [
 EventGraph {nodes [
   n1 = Event OnSquadIdChanged!() -> (OutputDelegate, then, InSquadData:Athena Remote Squad View Data Object Reference, InSquadId:Byte=0) {is=title K2Node_Event}
   > "SetText (Text)"(self=Get SquadId() -> (SquadId:Text Object Reference), InText="To Text (Byte)"!(Value:Byte=@n1.InSquadId) -> (ReturnValue:Text))
   n2 = Event OnSquadColorChanged!() -> (OutputDelegate, then, InSquadData:Athena Remote Squad View Data Object Reference, InSquadColor:Linear Color Structure) {is=title K2Node_Event}
   > SetBrushColor(self=Get SquadIdBorder() -> (SquadIdBorder:Common Border Object Reference), InBrushColor=@n2.InSquadColor)
   Event OnPlayerAdded!() -> (
    OutputDelegate
    then
    InSquadData:Athena Remote Squad View Data Object Reference
    InPlayerData:Athena Remote Player View Data Object Reference
   ) {is=title K2Node_Event}
   > RebuildPlayerList()
   n3 = Event SetSquadIdVisible!() -> (OutputDelegate, then, bSquadIdVisible:bool) {is=title K2Node_Event}
   > SetVisibility(
    self=Get SquadIdSB() -> (SquadIdSB:Size Box Object Reference)
    InVisibility=Select(Option 0=Collapsed, Option 1=SelfHitTestInvisible, Index=@n3.bSquadIdVisible)
   )
  ]}
 RebuildPlayerList {nodes [
   RebuildPlayerList() {is=title K2Node_FunctionEntry}
   > Reset!(
    execute
    self:Dynamic Entry Box Object Reference=Get PlayerBox() -> (PlayerBox:Fort Dynamic Entry Box Object Reference)
    bDeleteWidgets:bool=false
   ) -> (then)
   > n1 = For Each Loop(Array:Array of Athena Remote Player View Data Object References=GetPlayerDatas!(self:Athena Remote Squad View Data Object Reference=Get SquadData() -> (SquadData:Athena Remote Squad View Data Weak object)) -> (ReturnValue:Array of Athena Remote Player View Data Object References)) -> (Array Element:Athena Remote Player View Data Object Reference)
   InvalidateLayoutAndVolatility(execute=@n1.Completed, self:Widget Object Reference)
   n2 = Create Entry of Class!(
    execute=@n1.LoopBody
    self:Dynamic Entry Box Object Reference=Get PlayerBox()
    EntryClass:User Widget Class Reference=/Game/Athena/UI/Spectator/EventMatchInfo/AthenaEventMatchScoreboardPlayerWidget.AthenaEventMatchScoreboardPlayerWidget_C
   ) -> (then, ReturnValue:Athena Event Match Scoreboard Player Widget Object Reference)
   > n3 = Cast To AthenaEventMatchScoreboardPlayerWidget!(execute, Object=@n2) -> (then, CastFailed, AsAthena Event Match Scoreboard Player Widget:Athena Event Match Scoreboard Player Widget Object Reference) {is=title K2Node_DynamicCast}
   > SetPlayerData!(execute, self:Athena Event Match Player Widget Base Object Reference=@n3, InPlayerData:Athena Remote Player View Data Object Reference=@n1) -> (then)
  ]}
]
variables []
widgets {
 info {
  inheritedWidgetCount=1
  parentClass=/Script/FortniteUI.AthenaEventMatchScoreboardSquadWidget
  rootWidgetClass=&HorizontalBox
  widgetCount=6
 }
 widgets [
  HorizontalBox_1 {widgetClassPath=&HorizontalBox, widget=HorizontalBox_1}
  SquadIdSB {
   widgetClassPath=&SizeBox
   bIsVariable=true
   parent=HorizontalBox_1
   widget=SquadIdSB
   properties {bOverride_WidthOverride=true, widthOverride=50}
   slot {class=HorizontalBoxSlot}
  }
  SquadIdBorder {
   widgetClassPath=&CommonBorder
   bIsVariable=true
   parent=SquadIdSB
   widget=SquadIdBorder
   properties {
    horizontalAlignment=HAlign_Center
    verticalAlignment=VAlign_Center
    background {
     bIsDynamicallyLoaded=false
     drawAs=Box
     imageType=NoImage
     mirroring=NoMirror
     resourceName=None
     resourceObject=null
     tiling=NoTile
     imageSize=<32 32>
     margin=<ltrb 0 0 0 0>
     outlineSettings {
      bUseBrushTransparency=false
      roundingType=HalfHeightRadius
      width=0
      color {colorUseRule=UseColor_Specified, specifiedColor=<rgba 0 0 0 0>}
      cornerRadii=<0 0 0 0>
     }
     tintColor {colorUseRule=UseColor_Specified, specifiedColor=<rgba 1 1 1 1>}
     uVRegion {bIsValid=false, max=<0 0>, min=<0 0>}
    }
    brushColor=<rgba 1 1 0 1>
    padding=<ltrb 0 3 0 0>
   }
   slot {class=SizeBoxSlot}
  }
  SquadId {
   widgetClassPath=&TextBlock
   bIsVariable=true
   parent=SquadIdBorder
   widget=SquadId
   properties {
    text=1
    visibility=SelfHitTestInvisible
    colorAndOpacity {colorUseRule=UseColor_Specified, specifiedColor=<rgba 0 0 0 0.4>}
    font {
     bForceMonospaced=false
     bMaterialIsStencil=false
     fontMaterial=null
     fontObject=&Burbank
     letterSpacing=0
     monospacedWidth=1
     size=16
     skewAmount=0
     typefaceFontName=BurbankRegular-Black
     outlineSettings {
      bApplyOutlineToDropShadows=false
      bMiteredCorners=false
      bSeparateFillAlpha=false
      outlineMaterial=null
      outlineSize=0
      outlineColor=<rgba 0 0 0 1>
     }
    }
   }
   slot {class=BorderSlot, properties {horizontalAlignment=HAlign_Center, verticalAlignment=VAlign_Center, padding=<ltrb 0 3 0 0>}}
  }
  PlayerBox {
   widgetClassPath=/Script/FortniteUI.FortDynamicEntryBox
   bIsVariable=true
   parent=HorizontalBox_1
   widget=PlayerBox
   properties {
    entryBoxType=Vertical
    entryWidgetClass=&AthenaEventMatchScoreboardPlayerWidget_C
    maximumEntries=4
    numDesignerPreviewEntries=2
    entrySpacing=<0 3>
   }
   slot {class=HorizontalBoxSlot}
  }
  TutorialHighlight {widgetClassPath=&TutorialHighlight}
 ]
}
