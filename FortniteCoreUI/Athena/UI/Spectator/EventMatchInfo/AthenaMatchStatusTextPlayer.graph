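# Text widget for the spectator event match info that displays a top player's name or status with animated emphasis.
# SetText stores the text and can defer applying it; wrapper functions expose font and color, which PreConstruct also applies.
# Highlight events toggle a font material by rebuilding the font info.
# NewTopPlayer blinks the text, tints it gold and flashes staggered lens flares, while TopPlayerDead scales in a red strikethrough; neither is played here.
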WidgetBlueprint /FortniteCoreUI/Athena/UI/Spectator/EventMatchInfo/AthenaMatchStatusTextPlayer
animations [
 NewTopPlayer {
  binding_count=6
  display_label=""
  length_seconds=3
  object=&self:NewTopPlayer
  sequence {
   bindings [
    Strikethrough {
     id=#1
     possessed_class=Image
     tracks [{
       type=MovieSceneByteTrack
       display_name=Visibility
       sections [{type=MovieSceneByteSection, end_frame=60, start_frame=0, channels [None {type=MovieSceneScriptingByteChannel, const [2,1]}]}]
      }]
    }
    LensFlare4 {
     id=#2
     possessed_class=Image
     tracks [
      {
       type=MovieSceneColorTrack
       display_name=Color and Opacity
       sections [{
         type=MovieSceneColorSection
         end_frame=60
         start_frame=0
         channels [
          Color.R {type=MovieSceneScriptingFloatChannel, const [1,1]}
          Color.G {type=MovieSceneScriptingFloatChannel, const [1,1]}
          Color.B {type=MovieSceneScriptingFloatChannel, const [1,1]}
          Color.A {type=MovieSceneScriptingFloatChannel, thinned_from=6, keys [[0,0],[37,0],[44,1],[50,0],[60,0]]}
         ]
        }]
      }
      {
       type=MovieSceneByteTrack
       display_name=Visibility
       sections [{type=MovieSceneByteSection, end_frame=60, start_frame=0, channels [None {type=MovieSceneScriptingByteChannel, keys [[0,2],[13,0],[60,2]]}]}]
      }
     ]
    }
    LensFlare1 {
     id=#3
     possessed_class=Image
     tracks [
      {
       type=MovieSceneColorTrack
       display_name=Color and Opacity
       sections [{
         type=MovieSceneColorSection
         end_frame=60
         start_frame=0
         channels [
          Color.R {type=MovieSceneScriptingFloatChannel, const [1,1]}
          Color.G {type=MovieSceneScriptingFloatChannel, const [1,1]}
          Color.B {type=MovieSceneScriptingFloatChannel, const [1,1]}
          Color.A {type=MovieSceneScriptingFloatChannel, keys [[0,0],[13,0],[27,1],[39,0],[60,0]]}
         ]
        }]
      }
      {
       type=MovieSceneByteTrack
       display_name=Visibility
       sections [{type=MovieSceneByteSection, end_frame=60, start_frame=0, channels [None {type=MovieSceneScriptingByteChannel, keys [[0,2],[13,0],[60,2]]}]}]
      }
     ]
    }
    LensFlare2 {
     id=#4
     possessed_class=Image
     tracks [
      {
       type=MovieSceneColorTrack
       display_name=Color and Opacity
       sections [{
         type=MovieSceneColorSection
         end_frame=60
         start_frame=0
         channels [
          Color.R {type=MovieSceneScriptingFloatChannel, const [1,1]}
          Color.G {type=MovieSceneScriptingFloatChannel, const [1,1]}
          Color.B {type=MovieSceneScriptingFloatChannel, const [1,1]}
          Color.A {type=MovieSceneScriptingFloatChannel, thinned_from=6, keys [[0,0],[21,0],[26,1],[30,0],[60,0]]}
         ]
        }]
      }
      {
       type=MovieSceneByteTrack
       display_name=Visibility
       sections [{type=MovieSceneByteSection, end_frame=60, start_frame=0, channels [None {type=MovieSceneScriptingByteChannel, keys [[0,2],[13,0],[60,2]]}]}]
      }
     ]
    }
    TextBlock {
     id=#5
     possessed_class=TextBlock
     tracks [{
       type=MovieSceneColorTrack
       display_name=Color and Opacity
       sections [{
         type=MovieSceneColorSection
         end_frame=60
         start_frame=0
         channels [
          Color.R {type=MovieSceneScriptingFloatChannel, const [1,4]}
          Color.G {type=MovieSceneScriptingFloatChannel, keys [[0,1],[10,1],[50,0.79],[60,1]]}
          Color.B {type=MovieSceneScriptingFloatChannel, keys [[0,1],[10,1],[50,0.005],[60,1]]}
          Color.A {type=MovieSceneScriptingFloatChannel, keys [[0,1],[10,0],[13,1]]}
         ]
        }]
      }]
    }
    LensFlare3 {
     id=#6
     possessed_class=Image
     tracks [
      {
       type=MovieSceneColorTrack
       display_name=Color and Opacity
       sections [{
         type=MovieSceneColorSection
         end_frame=60
         start_frame=0
         channels [
          Color.R {type=MovieSceneScriptingFloatChannel, const [1,1]}
          Color.G {type=MovieSceneScriptingFloatChannel, const [1,1]}
          Color.B {type=MovieSceneScriptingFloatChannel, const [1,1]}
          Color.A {type=MovieSceneScriptingFloatChannel, thinned_from=6, keys [[0,0],[26,0],[38,1],[49,0],[60,0]]}
         ]
        }]
      }
      {
       type=MovieSceneByteTrack
       display_name=Visibility
       sections [{type=MovieSceneByteSection, end_frame=60, start_frame=0, channels [None {type=MovieSceneScriptingByteChannel, keys [[0,2],[13,0],[60,2]]}]}]
      }
     ]
    }
   ]
   tracks [{type=MovieSceneEventTrack, sections [{type=MovieSceneEventSection}]}]
  }
 }
 TopPlayerDead {
  binding_count=5
  display_label=""
  length_seconds=0.5
  object=&self:TopPlayerDead
  sequence {bindings [
    LensFlare4 {
     id=#7
     possessed_class=Image
     tracks [{
       type=MovieSceneByteTrack
       display_name=Visibility
       sections [{type=MovieSceneByteSection, end_frame=0, start_frame=0, channels [None {type=MovieSceneScriptingByteChannel, const [2,1]}]}]
      }]
    }
    Strikethrough {
     id=#8
     possessed_class=Image
     tracks [
      {
       type=MovieScene2DTransformTrack
       sections [{
         type=MovieScene2DTransformSection
         end_frame=10
         start_frame=0
         channels [
          Translation.X {type=MovieSceneScriptingFloatChannel}
          Translation.Y {type=MovieSceneScriptingFloatChannel}
          Angle {type=MovieSceneScriptingFloatChannel}
          Scale.X {type=MovieSceneScriptingFloatChannel, keys [[0.0003,0],[5.0003,1]]}
          Scale.Y {type=MovieSceneScriptingFloatChannel, keys [[0.0003,0],[5.0003,1]]}
          Shear.X {type=MovieSceneScriptingFloatChannel}
          Shear.Y {type=MovieSceneScriptingFloatChannel}
         ]
        }]
      }
      {
       type=MovieSceneByteTrack
       display_name=Visibility
       sections [{type=MovieSceneByteSection, end_frame=10, start_frame=0, channels [None {type=MovieSceneScriptingByteChannel, const [0,1]}]}]
      }
     ]
    }
    LensFlare3 {
     id=#9
     possessed_class=Image
     tracks [{
       type=MovieSceneByteTrack
       display_name=Visibility
       sections [{type=MovieSceneByteSection, end_frame=0, start_frame=0, channels [None {type=MovieSceneScriptingByteChannel, const [2,1]}]}]
      }]
    }
    LensFlare1 {
     id=#10
     possessed_class=Image
     tracks [{
       type=MovieSceneByteTrack
       display_name=Visibility
       sections [{type=MovieSceneByteSection, end_frame=10, start_frame=0, channels [None {type=MovieSceneScriptingByteChannel, const [2,1]}]}]
      }]
    }
    LensFlare2 {
     id=#11
     possessed_class=Image
     tracks [{
       type=MovieSceneByteTrack
       display_name=Visibility
       sections [{type=MovieSceneByteSection, end_frame=0, start_frame=0, channels [None {type=MovieSceneScriptingByteChannel, const [2,1]}]}]
      }]
    }
   ]}
 }
]
graphs [
 ApplyText {nodes [
   ApplyText() {is=title K2Node_FunctionEntry}
   > "SetText (Text)"(self=Get TextBlock() -> (TextBlock:Text Object Reference), InText=Get TextToDisplay() -> (TextToDisplay:Text))
  ]}
 EventGraph {nodes [
   HighlightMaterialOn() {is=title K2Node_CustomEvent}
   > n1 = Set ShowMaterialHighlight(ShowMaterialHighlight:bool=true) -> (Output_Get:bool)
   HighlightMaterialControl(execute=[@n1.then, @n2.then])
   HighlightMaterialOff() {is=title K2Node_CustomEvent}
   > n2 = Set ShowMaterialHighlight(ShowMaterialHighlight=false)
   Event PreConstruct()
   n3 = Get TextBlock()
   > "SetText (Text)"(self=@n3, InText=Get TextToDisplay())
   > SetFont(self=@n3, InFontInfo=Get Font() -> (Font:Slate Font Info Structure))
   > SetColorAndOpacity(self=@n3, InColorAndOpacity=Get Color() -> (Color:Slate Color Structure))
   ApplyText() {is=title K2Node_CallFunction}
   ApplyNewText() {is=title K2Node_CustomEvent}
   Event Construct()
   Event Tick()
  ]}
 GetFont {nodes [
   GetFont() {is=title K2Node_FunctionEntry}
   > Return Node!(execute, Output:Slate Font Info Structure=Get Font!(self:Text Object Reference=Get TextBlock()) -> (Font))
  ]}
 HighlightMaterialControl {nodes [
   HighlightMaterialControl() {is=title K2Node_FunctionEntry}
   > n1 = Branch(Condition=Get ShowMaterialHighlight() -> (ShowMaterialHighlight:bool))
   n2 = Get TextBlock()
   n3 = Get!(self:Text Object Reference=@n2) -> (
    Font_FontObject:Object Reference
    Font_FontMaterial:Object Reference
    Font_OutlineSettings:Font Outline Settings Structure
    Font_TypefaceFontName:Name
    Font_Size:float=~
    Font_LetterSpacing:int
    Font_SkewAmount:float=~
    Font_bForceMonospaced:bool=~
    Font_bMaterialIsStencil:bool=~
    Font_MonospacedWidth:float=~
   )
   > SetFont!(
    execute
    self=@n2
    InFontInfo_FontObject:Object Reference=@n3
    InFontInfo_FontMaterial:Object Reference=Get HighlightMaterial() -> (HighlightMaterial:Material Instance Object Reference)
    InFontInfo_OutlineSettings:Font Outline Settings Structure=@n3.Font_OutlineSettings
    InFontInfo_TypefaceFontName:Name=@n3.Font_TypefaceFontName
    InFontInfo_Size:float=@n3.Font_Size
    InFontInfo_LetterSpacing:int=0
    InFontInfo_SkewAmount:float=0
    InFontInfo_bForceMonospaced:bool=false
    InFontInfo_bMaterialIsStencil:bool=false
    InFontInfo_MonospacedWidth:float=1
   ) -> (then)
   SetFont(
    execute=@n1.else
    self=@n2
    InFontInfo_FontObject=@n3
    InFontInfo_OutlineSettings=@n3.Font_OutlineSettings
    InFontInfo_TypefaceFontName=@n3.Font_TypefaceFontName
    InFontInfo_Size=@n3.Font_Size
    InFontInfo_LetterSpacing=0
    InFontInfo_SkewAmount=0
    InFontInfo_bForceMonospaced=false
    InFontInfo_bMaterialIsStencil=false
    InFontInfo_MonospacedWidth=1
   )
  ]}
 SetFont {nodes [
   n1 = SetFont!() -> (then, FontIn:Slate Font Info Structure) {is=title K2Node_FunctionEntry}
   > SetFont!(execute, self=Get TextBlock(), InFontInfo:Slate Font Info Structure=@n1) -> (then) {is=title K2Node_CallFunction}
  ]}
 SetOpacityAndColor {nodes [
   n1 = SetOpacityAndColor!() -> (then, ColorIn:Slate Color Structure) {is=title K2Node_FunctionEntry}
   > SetColorAndOpacity(self=Get TextBlock(), InColorAndOpacity=@n1)
  ]}
 SetText {nodes [
   n1 = SetText!() -> (then, InText:Text, ApplyImmediately:bool=~) {is=title K2Node_FunctionEntry}
   > Set TextToDisplay(TextToDisplay:Text=@n1) -> (Output_Get:Text)
   > Branch(Condition=@n1.ApplyImmediately)
   > ApplyText()
  ]}
]
variables [
 Color {type=SlateColor, default {colorUseRule=UseColor_Specified, specifiedColor=<rgba 1 1 1 1>}}
 Font {
  type=SlateFontInfo
  default {
   bForceMonospaced=false
   bMaterialIsStencil=false
   fontMaterial=null
   fontObject=&Burbank
   letterSpacing=0
   monospacedWidth=1
   size=42
   skewAmount=0
   typefaceFontName=BurbankRegular-Black
   outlineSettings {
    bApplyOutlineToDropShadows=false
    bMiteredCorners=false
    bSeparateFillAlpha=false
    outlineMaterial=null
    outlineSize=0
    outlineColor=<rgba 0 0 0 1>
   }
  }
 }
 HighlightMaterial {type=MaterialInstance, default=/Game/Athena/UI/Spectator/EventMatchInfo/M_PlayerHighlight}
 ShowMaterialHighlight {type=boolean}
 TextToDisplay {type=string, default=EXAMPLE TEXT}
]
widgets {
 info {inheritedWidgetCount=0, parentClass=&UserWidget, rootWidgetClass=&Overlay, widgetCount=7}
 widgets [
  Overlay_0 {widgetClassPath=&Overlay, widget=Overlay_0}
  TextBlock {
   widgetClassPath=&TextBlock
   bIsVariable=true
   parent=Overlay_0
   widget=TextBlock
   properties {
    text=EXAMPLE TEXT
    font {
     bForceMonospaced=false
     bMaterialIsStencil=false
     fontMaterial=null
     fontObject=&Burbank
     letterSpacing=0
     monospacedWidth=1
     size=42
     skewAmount=0
     typefaceFontName=BurbankRegular-Black
     outlineSettings {
      bApplyOutlineToDropShadows=false
      bMiteredCorners=false
      bSeparateFillAlpha=false
      outlineMaterial=null
      outlineSize=0
      outlineColor=<rgba 0 0 0 1>
     }
    }
   }
   slot {class=OverlaySlot, properties {horizontalAlignment=HAlign_Center, verticalAlignment=VAlign_Center}}
  }
  LensFlare1 {
   widgetClassPath=&Image
   bIsVariable=true
   parent=Overlay_0
   widget=LensFlare1
   properties {
    visibility=Hidden
    brush {
     bIsDynamicallyLoaded=false
     drawAs=Image
     imageType=NoImage
     mirroring=NoMirror
     resourceName=None
     resourceObject=/Game/UI/ItemReceived/Materials/M_EFX_CameraFlare_01
     tiling=NoTile
     imageSize=<64 64>
     margin=<ltrb 0 0 0 0>
     outlineSettings {
      bUseBrushTransparency=false
      roundingType=HalfHeightRadius
      width=0
      color {colorUseRule=UseColor_Specified, specifiedColor=<rgba 0 0 0 0>}
      cornerRadii=<0 0 0 0>
     }
     tintColor {colorUseRule=UseColor_Specified, specifiedColor=<rgba 1 1 1 1>}
     uVRegion {bIsValid=false, max=<0 0>, min=<0 0>}
    }
    renderTransform {angle=0, scale=<1 1>, shear=<0 0>, translation=<-74 -26>}
   }
   slot {class=OverlaySlot, properties {horizontalAlignment=HAlign_Center, verticalAlignment=VAlign_Center}}
  }
  LensFlare2 {
   widgetClassPath=&Image
   bIsVariable=true
   parent=Overlay_0
   widget=LensFlare2
   properties {
    visibility=Hidden
    brush {
     bIsDynamicallyLoaded=false
     drawAs=Image
     imageType=NoImage
     mirroring=NoMirror
     resourceName=None
     resourceObject=&M_EFX_CameraFlare_01
     tiling=NoTile
     imageSize=<48 48>
     margin=<ltrb 0 0 0 0>
     outlineSettings {
      bUseBrushTransparency=false
      roundingType=HalfHeightRadius
      width=0
      color {colorUseRule=UseColor_Specified, specifiedColor=<rgba 0 0 0 0>}
      cornerRadii=<0 0 0 0>
     }
     tintColor {colorUseRule=UseColor_Specified, specifiedColor=<rgba 1 1 1 1>}
     uVRegion {bIsValid=false, max=<0 0>, min=<0 0>}
    }
    renderTransform {angle=0, scale=<1 1>, shear=<0 0>, translation=<-36 14>}
   }
   slot {class=OverlaySlot, properties {horizontalAlignment=HAlign_Center, verticalAlignment=VAlign_Center}}
  }
  LensFlare3 {
   widgetClassPath=&Image
   bIsVariable=true
   parent=Overlay_0
   widget=LensFlare3
   properties {
    visibility=Hidden
    brush {
     bIsDynamicallyLoaded=false
     drawAs=Image
     imageType=NoImage
     mirroring=NoMirror
     resourceName=None
     resourceObject=&M_EFX_CameraFlare_01
     tiling=NoTile
     imageSize=<48 48>
     margin=<ltrb 0 0 0 0>
     outlineSettings {
      bUseBrushTransparency=false
      roundingType=HalfHeightRadius
      width=0
      color {colorUseRule=UseColor_Specified, specifiedColor=<rgba 0 0 0 0>}
      cornerRadii=<0 0 0 0>
     }
     tintColor {colorUseRule=UseColor_Specified, specifiedColor=<rgba 1 1 1 1>}
     uVRegion {bIsValid=false, max=<0 0>, min=<0 0>}
    }
    renderTransform {angle=0, scale=<1 1>, shear=<0 0>, translation=<80 -12>}
   }
   slot {class=OverlaySlot, properties {horizontalAlignment=HAlign_Center, verticalAlignment=VAlign_Center}}
  }
  LensFlare4 {
   widgetClassPath=&Image
   bIsVariable=true
   parent=Overlay_0
   widget=LensFlare4
   properties {
    visibility=Collapsed
    brush {
     bIsDynamicallyLoaded=false
     drawAs=Image
     imageType=NoImage
     mirroring=NoMirror
     resourceName=None
     resourceObject=&M_EFX_CameraFlare_01
     tiling=NoTile
     imageSize=<64 64>
     margin=<ltrb 0 0 0 0>
     outlineSettings {
      bUseBrushTransparency=false
      roundingType=HalfHeightRadius
      width=0
      color {colorUseRule=UseColor_Specified, specifiedColor=<rgba 0 0 0 0>}
      cornerRadii=<0 0 0 0>
     }
     tintColor {colorUseRule=UseColor_Specified, specifiedColor=<rgba 1 1 1 1>}
     uVRegion {bIsValid=false, max=<0 0>, min=<0 0>}
    }
    renderTransform {angle=0, scale=<1 1>, shear=<0 0>, translation=<42 -3>}
   }
   slot {class=OverlaySlot, properties {horizontalAlignment=HAlign_Center, verticalAlignment=VAlign_Center}}
  }
  Strikethrough {
   widgetClassPath=&Image
   bIsVariable=true
   parent=Overlay_0
   widget=Strikethrough
   properties {
    visibility=Hidden
    brush {
     bIsDynamicallyLoaded=false
     drawAs=Image
     imageType=NoImage
     mirroring=NoMirror
     resourceName=None
     resourceObject=/Game/Athena/UI/Spectator/EventMatchInfo/slashMarker
     tiling=NoTile
     imageSize=<256 64>
     margin=<ltrb 0 0 0 0>
     outlineSettings {
      bUseBrushTransparency=false
      roundingType=HalfHeightRadius
      width=0
      color {colorUseRule=UseColor_Specified, specifiedColor=<rgba 0 0 0 0>}
      cornerRadii=<0 0 0 0>
     }
     tintColor {colorUseRule=UseColor_Specified, specifiedColor=<rgba 1 1 1 1>}
     uVRegion {bIsValid=false, max=<0 0>, min=<0 0>}
    }
    colorAndOpacity=<rgba 1 0.165 0.165 0.9>
    renderTransformPivot=<0 1>
   }
   slot {class=OverlaySlot, properties {horizontalAlignment=HAlign_Center, verticalAlignment=VAlign_Center}}
  }
 ]
}
